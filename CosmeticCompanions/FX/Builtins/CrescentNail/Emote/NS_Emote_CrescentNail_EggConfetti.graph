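# Niagara system for the egg confetti effect of the CrescentNail cosmetic companion emote.
# It combines three emitters: AirglowLWE, SmokeLWE and Confetti_Low, the first two named as lightweight emitters.
# Its listed modules are SystemState and a day-sequence time-of-day color utility.
# User parameters expose front-end and front-end preview flags plus a normalized notify progress float; renderers and parameter wiring are not recorded.
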NiagaraSystem /CosmeticCompanions/FX/Builtins/CrescentNail/Emote/NS_Emote_CrescentNail_EggConfetti
dependencies {
 usedDataInterfaces []
 usedDynamicInputs []
 usedModules [/Game/Effects/Niagara/Modules/Utility/DaySequenceTimeOfDayColor, &SystemState]
}
emitters {AirglowLWE {rendererClasses [], renderers []}, Confetti_Low {rendererClasses [], renderers []}, SmokeLWE {rendererClasses [], renderers []}}
summary {
 systemName=NS_Emote_CrescentNail_EggConfetti
 emitters [{emitterName=AirglowLWE, rendererClasses []}, {emitterName=SmokeLWE, rendererClasses []}, {emitterName=Confetti_Low, rendererClasses []}]
 userVariables [
  User.bIsFrontEnd {type=NiagaraBool}
  User.bIsFrontEndPreview {type=NiagaraBool}
  User.NormalizedNotifyProgress {type=NiagaraFloat, defaultValue=0, description=""}
 ]
}
